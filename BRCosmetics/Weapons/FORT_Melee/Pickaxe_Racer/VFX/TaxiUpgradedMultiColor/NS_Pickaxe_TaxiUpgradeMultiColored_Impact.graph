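# Niagara impact burst for the multicolored TaxiUpgraded variant of the Racer pickaxe, with every emitter lasting well under a second.
# Corona, sprite and mesh layers cycle through a rainbow gradient over particle life.
# A flipbook core flash adds a light, alongside panned spike streak meshes and pink mesh shards scattered by curl noise, drag and gravity.
# User.VariantWhite blends all six emitters' colors toward white or grayscale for an alternate style.
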NiagaraSystem /BRCosmetics/Weapons/FORT_Melee/Pickaxe_Racer/VFX/TaxiUpgradedMultiColor/NS_Pickaxe_TaxiUpgradeMultiColored_Impact
dependencies {
 usedDataInterfaces [&NiagaraDataInterfaceColorCurve, &NiagaraDataInterfaceCurve, &NiagaraDataInterfaceVectorCurve]
 usedDynamicInputs []
 usedModules [&SystemState]
}
emitters {
 Corona_v2 {
  emitterUpdateScript [
   &EmitterState()
   &SpawnBurst_Instantaneous(Spawn Count=1, Use Spawn Probability=false)
  ]
  particleSpawnScript [
   /Niagara/Modules/Spawn/Initialization/InitializeParticle(
    Lifetime=0.1
    Module.Write Color=true
    Module.Write Lifetime=true
    Module.Write Mass=true
    Module.Write Material Random=true
    Module.Write Position=true
    Module.Write Scale=false
    Module.Write SpriteRotation=true
    Module.Write SpriteSize=true
    Sprite Rotation=&NormalizedAngleToDegrees(NormalizedAngle=&UniformRangedFloat(Minimum=0, Maximum=1))
    Sprite Size=&Vector2DFromFloat(Value=250)
   )
   &CameraOffset(Camera Offset Amount=25)
  ]
  particleUpdateScript [
   &ParticleState()
   /SetVariables(Particles.GradientColor:Vector3f=&MakeVectorFromLinearColorRGB(Color=&MakeLinearColorFromVectorAndFloat(
      "Vector (RGB)"=&MakeVectorFromLinearColorRGB(Color=&LerpLinearColors(
        StartColor=&ColorFromCurve(ColorCurve=DataInterface() {propertyValues {
           AlphaCurve {keys [<at -0.0023 1 RCIM_Cubic>,<at 0.9977 1 RCIM_Cubic>]}
           BlueCurve {keys [<at 0.1175 0 RCIM_Cubic>,<at 0.2682 0>,<at 0.4507 0>,<at 0.5951 1>,<at 0.7691 1.5>,<at 0.9071 1.5678 RCIM_Cubic>]}
           GreenCurve {keys [<at 0.1175 0.039 RCIM_Cubic>,<at 0.2682 0.8559>,<at 0.4507 1>,<at 0.5951 0.6207>,<at 0.7691 0.4532>,<at 0.9071 0 RCIM_Cubic>]}
           RedCurve {keys [<at 0.1175 1 RCIM_Cubic>,<at 0.2682 2>,<at 0.4507 0.3535>,<at 0.5951 0.0629>,<at 0.7691 0>,<at 0.9071 2 RCIM_Cubic>]}
          }})
        EndColor=&ColorFromCurve(ColorCurve=DataInterface() {propertyValues {
           AlphaCurve {keys [<at 0 1 RCIM_Cubic>,<at 0.9977 1 RCIM_Cubic>]}
           BlueCurve {keys [<at 0.1175 0.3073 RCIM_Cubic>,<at 0.2682 0.6024>,<at 0.4507 0.4809>,<at 0.5951 0.0816>,<at 0.7691 0.7378>,<at 0.9071 0.2865 RCIM_Cubic>]}
           GreenCurve {keys [<at 0.1175 0.3073 RCIM_Cubic>,<at 0.2682 0.6024>,<at 0.4507 0.4809>,<at 0.5951 0.0816>,<at 0.7691 0.7378>,<at 0.9071 0.2865 RCIM_Cubic>]}
           RedCurve {keys [<at 0.1175 0.3073 RCIM_Cubic>,<at 0.2682 0.6024>,<at 0.4507 0.4809>,<at 0.5951 0.0816>,<at 0.7691 0.7378>,<at 0.9071 0.2865 RCIM_Cubic>]}
          }})
        LerpFactor=$User.VariantWhite
       ))
      "Float (Alpha)"=1
     ))) {isSetParameters=true}
   &ScaleSpriteSize(Scale Factor=&Vector2DFromFloat(Value=&FloatFromCurve(FloatCurve=DataInterface() {propertyValues {Curve {keys [
          {arriveTangent=2.8962, interpMode=RCIM_Cubic, leaveTangent=2.8962, tangentMode=RCTM_User, time=-0.0013, value=0.29}
          {arriveTangent=1.7111, interpMode=RCIM_Cubic, leaveTangent=1.7111, tangentMode=RCTM_User, time=0.15, value=0.66}
          <at 0.9981 1.0141>
         ]}}})))
   /SetVariables(Particles.Color=&MakeLinearColorFromVectorAndFloat(
     "Vector (RGB)"=&Lerp_Vector(
      A=&Multiply_VectorByFloat(Vector=$Particles.GradientColor, Float=1)
      B=&Multiply_VectorByFloat(Vector=&UniformRangedVector(Minimum=$Particles.GradientColor, Maximum=$Particles.GradientColor), Float=2)
      Alpha=&FloatFromCurve(FloatCurve=DataInterface() {propertyValues {Curve {keys [
           {interpMode=RCIM_Cubic, tangentMode=RCTM_User, time=-0.0056, value=0.0478}
           {arriveTangent=0.0255, interpMode=RCIM_Cubic, leaveTangent=0.0255, tangentMode=RCTM_User, time=0.241, value=0.0789}
           {arriveTangent=0.1228, interpMode=RCIM_Cubic, leaveTangent=0.1228, tangentMode=RCTM_User, time=0.5453, value=0.6822}
           {interpMode=RCIM_Cubic, tangentMode=RCTM_User, time=1.0011, value=0.002}
          ]}}})
     )
     "Float (Alpha)"=&FloatFromCurve(
      FloatCurve=DataInterface() {propertyValues {Curve {keys [
          {interpMode=RCIM_Cubic, tangentMode=RCTM_User, time=-0.0063, value=0.0184}
          {arriveTangent=1.3376, interpMode=RCIM_Cubic, leaveTangent=1.3376, tangentMode=RCTM_User, time=0.096, value=0.9808}
          {arriveTangent=0.3112, interpMode=RCIM_Cubic, leaveTangent=0.3112, tangentMode=RCTM_User, time=0.5407, value=1.3419}
          {arriveTangent=-3.7119, interpMode=RCIM_Cubic, leaveTangent=-3.7119, tangentMode=RCTM_User, time=0.7242, value=0.3749}
          {interpMode=RCIM_Cubic, tangentMode=RCTM_User, time=1.0011, value=0.002}
         ]}}}
      Scale Curve=0.25
     )
    )) {isSetParameters=true}
   /SetVariables(Particles.Color2=&MakeLinearColorFromVectorAndFloat(
     "Vector (RGB)"=&Lerp_Vector(
      A=&Multiply_VectorByFloat(Vector=$Particles.GradientColor, Float=1)
      B=&Multiply_VectorByFloat(Vector=&UniformRangedVector(Minimum=$Particles.GradientColor, Maximum=$Particles.GradientColor), Float=2)
      Alpha=&FloatFromCurve(FloatCurve=DataInterface() {propertyValues {Curve {keys [
           {interpMode=RCIM_Cubic, tangentMode=RCTM_User, time=-0.0008, value=0.5048}
           {arriveTangent=0.0255, interpMode=RCIM_Cubic, leaveTangent=0.0255, tangentMode=RCTM_User, time=0.0802, value=0.5788}
           {arriveTangent=0.1228, interpMode=RCIM_Cubic, leaveTangent=0.1228, tangentMode=RCTM_User, time=0.5421, value=2.0946}
           {interpMode=RCIM_Cubic, tangentMode=RCTM_User, time=1.0011, value=0.002}
          ]}}})
     )
     "Float (Alpha)"=&FloatFromCurve(
      FloatCurve=DataInterface() {propertyValues {Curve {keys [
          {interpMode=RCIM_Cubic, tangentMode=RCTM_User, time=-0.0063, value=0.0184}
          {arriveTangent=0.3112, interpMode=RCIM_Cubic, leaveTangent=0.3112, tangentMode=RCTM_User, time=0.3734, value=0.0777}
          {arriveTangent=-1.0034, interpMode=RCIM_Cubic, leaveTangent=-1.0034, tangentMode=RCTM_User, time=0.6422, value=0.1616}
          {interpMode=RCIM_Cubic, tangentMode=RCTM_User, time=1.0011, value=0.002}
         ]}}}
      Scale Curve=0.1
     )
    )) {isSetParameters=true}
   &DynamicMaterialParameters(
    Index 0 Param 1=1
    Index 0 Param 2=0.5
    Index 0 Param 3=0.6
    Index 0 Param 4=1
    Module.Param0WriteEnabled=true
    Module.Param1WriteEnabled=true
    Module.Param2WriteEnabled=true
    Module.Param3WriteEnabled=true
   )
  ]
  rendererClasses [&NiagaraSpriteRendererProperties]
  renderers [{rendererClass=&NiagaraSpriteRendererProperties}]
 }
 FN_Mesh {
  emitterUpdateScript [
   &EmitterState()
   &SpawnBurst_Instantaneous(Spawn Count=1, Use Spawn Probability=false)
  ]
  particleSpawnScript [
   &InitializeParticle(
    Lifetime=0.15
    Mesh Scale=&Multiply_Vector(
     A=&VectorFromFloat(Value=0.25)
     B=&UniformRangedVector(Minimum=<1 3 3>, Maximum=<1 3 3>)
    )
    Module.Write Color=true
    Module.Write Lifetime=true
    Module.Write Mass=true
    Module.Write Material Random=true
    Module.Write Position=false
    Module.Write Scale=true
    Module.Write SpriteRotation=false
    Module.Write SpriteSize=false
    Position=&SimulationPosition()
   )
   &DynamicMaterialParameters(
    Index 0 Param 1=&UniformRangedFloat(Minimum=0, Maximum=1)
    Index 0 Param 2=&UniformRangedFloat(Minimum=0.5, Maximum=1.25)
    Index 0 Param 3=&UniformRangedFloat(Minimum=0.25, Maximum=1)
    Index 0 Param 4=1
    Module.Param0WriteEnabled=true
    Module.Param1WriteEnabled=true
    Module.Param2WriteEnabled=true
    Module.Param3WriteEnabled=false
   )
   &InitialMeshOrientation(
    Rotation=&RandomRangeVector(Minimum=<0 0 0>, Maximum=<1 1 1>)
    Use Orientation Vector=true
    Use Rotation Vector=true
   )
   /SetVariables() {isSetParameters=true}
  ]
  particleUpdateScript [
   &ParticleState()
   &CameraOffset(Camera Offset Amount=30) {module=CameraOffset001}
   /SetVariables(Particles.GradientColor:Vector3f=&MakeVectorFromLinearColorRGB(Color=&MakeLinearColorFromVectorAndFloat(
      "Vector (RGB)"=&MakeVectorFromLinearColorRGB(Color=&MakeLinearColorFromVectorAndFloat(
        "Vector (RGB)"=&MakeVectorFromLinearColorRGB(Color=&LerpLinearColors(
          StartColor=&ColorFromCurve(ColorCurve=DataInterface() {propertyValues {
             AlphaCurve {keys [<at -0.0023 1 RCIM_Cubic>,<at 0.9977 1 RCIM_Cubic>]}
             BlueCurve {keys [<at 0.1175 0 RCIM_Cubic>,<at 0.2682 0>,<at 0.4507 0>,<at 0.5951 1>,<at 0.7691 1.5>,<at 0.9071 1.5678 RCIM_Cubic>]}
             GreenCurve {keys [<at 0.1175 0.039 RCIM_Cubic>,<at 0.2682 0.8559>,<at 0.4507 1>,<at 0.5951 0.6207>,<at 0.7691 0.4532>,<at 0.9071 0 RCIM_Cubic>]}
             RedCurve {keys [<at 0.1175 1 RCIM_Cubic>,<at 0.2682 2>,<at 0.4507 0.3535>,<at 0.5951 0.0629>,<at 0.7691 0>,<at 0.9071 2 RCIM_Cubic>]}
            }})
          EndColor=&ColorFromCurve(ColorCurve=DataInterface() {propertyValues {
             AlphaCurve {keys [<at 0 1 RCIM_Cubic>,<at 0.9977 1 RCIM_Cubic>]}
             BlueCurve {keys [<at 0.1175 0.3073 RCIM_Cubic>,<at 0.2682 0.6024>,<at 0.4507 0.4809>,<at 0.5951 0.0816>,<at 0.7691 0.7378>,<at 0.9071 0.2865 RCIM_Cubic>]}
             GreenCurve {keys [<at 0.1175 0.3073 RCIM_Cubic>,<at 0.2682 0.6024>,<at 0.4507 0.4809>,<at 0.5951 0.0816>,<at 0.7691 0.7378>,<at 0.9071 0.2865 RCIM_Cubic>]}
             RedCurve {keys [<at 0.1175 0.3073 RCIM_Cubic>,<at 0.2682 0.6024>,<at 0.4507 0.4809>,<at 0.5951 0.0816>,<at 0.7691 0.7378>,<at 0.9071 0.2865 RCIM_Cubic>]}
            }})
          LerpFactor=$User.VariantWhite
         ))
        "Float (Alpha)"=1
       ))
      "Float (Alpha)"=1
     ))) {isSetParameters=true}
   /SetVariables(Particles.Color=&MakeLinearColorFromVectorAndFloat(
     "Vector (RGB)"=&Lerp_Vector(
      A=&Multiply_VectorByFloat(Vector=$Particles.GradientColor, Float=1)
      B=&Multiply_VectorByFloat(Vector=$Particles.GradientColor, Float=1)
      Alpha=0.5
     )
     "Float (Alpha)"=&FloatFromCurve(
      FloatCurve=DataInterface() {propertyValues {Curve {keys [
          {interpMode=RCIM_Cubic, tangentMode=RCTM_User, time=-0.0063, value=0.0184}
          {arriveTangent=-0.8483, interpMode=RCIM_Cubic, leaveTangent=-0.8483, tangentMode=RCTM_User, time=0.1718, value=1.7147}
          {arriveTangent=-5.0978, interpMode=RCIM_Cubic, leaveTangent=-5.0978, tangentMode=RCTM_User, time=0.2818, value=1.2143}
          {arriveTangent=-0.2783, interpMode=RCIM_Cubic, leaveTangent=-0.2783, tangentMode=RCTM_User, time=0.4294, value=1.1587}
          {arriveTangent=0.3112, interpMode=RCIM_Cubic, leaveTangent=0.3112, tangentMode=RCTM_User, time=0.5407, value=1.3419}
          {arriveTangent=-3.7119, interpMode=RCIM_Cubic, leaveTangent=-3.7119, tangentMode=RCTM_User, time=0.7242, value=0.3749}
          {interpMode=RCIM_Cubic, tangentMode=RCTM_User, time=1.0011, value=0.002}
         ]}}}
      Scale Curve=0.25
     )
    )) {isSetParameters=true}
   /SetVariables(
    Particles.DynamicMaterialParameter:Vector4f=&MakeFloat4FromVectorAndFloat("Vector (RGB)"=&VectorFromFloat(Value=1), "Float (Alpha)"=1)
    Particles.DynamicMaterialParameter1:Vector4f=&MakeFloat4FromVectorAndFloat(
     "Vector (RGB)"=&VectorFromFloat(Value=1)
     "Float (Alpha)"=&UniformRangedFloat(Minimum=0, Maximum=1)
    )
   ) {isSetParameters=true}
  ]
  rendererClasses [&NiagaraMeshRendererProperties]
  renderers [{rendererClass=&NiagaraMeshRendererProperties, rendererIndex=0}]
 }
 FN_Sprite001 {
  emitterUpdateScript [
   &EmitterState()
   &SpawnBurst_Instantaneous(Spawn Count=1, Use Spawn Probability=false)
  ]
  particleSpawnScript [
   &InitializeParticle(
    Lifetime=0.18
    Module.Write Color=true
    Module.Write Lifetime=true
    Module.Write Mass=true
    Module.Write Material Random=true
    Module.Write Position=false
    Module.Write Scale=false
    Module.Write SpriteRotation=true
    Module.Write SpriteSize=true
    Position=&SimulationPosition()
    Sprite Rotation=&UniformRangedFloat(Minimum=0, Maximum=360)
    Sprite Size=&Vector2DFromFloat(Value=500)
   )
   &CameraOffset(Camera Offset Amount=25)
  ]
  particleUpdateScript [
   &ParticleState()
   /SetVariables(Particles.GradientColor:Vector3f=&MakeVectorFromLinearColorRGB(Color=&MakeLinearColorFromVectorAndFloat(
      "Vector (RGB)"=&MakeVectorFromLinearColorRGB(Color=&LerpLinearColors(
        StartColor=&ColorFromCurve(ColorCurve=DataInterface() {propertyValues {
           AlphaCurve {keys [<at -0.0023 1 RCIM_Cubic>,<at 0.9977 1 RCIM_Cubic>]}
           BlueCurve {keys [<at 0.1175 0 RCIM_Cubic>,<at 0.2682 0>,<at 0.4507 0>,<at 0.5951 1>,<at 0.7691 1.5>,<at 0.9071 1.5678 RCIM_Cubic>]}
           GreenCurve {keys [<at 0.1175 0.039 RCIM_Cubic>,<at 0.2682 0.8559>,<at 0.4507 1>,<at 0.5951 0.6207>,<at 0.7691 0.4532>,<at 0.9071 0 RCIM_Cubic>]}
           RedCurve {keys [<at 0.1175 1 RCIM_Cubic>,<at 0.2682 2>,<at 0.4507 0.3535>,<at 0.5951 0.0629>,<at 0.7691 0>,<at 0.9071 2 RCIM_Cubic>]}
          }})
        EndColor=&ColorFromCurve(ColorCurve=DataInterface() {propertyValues {
           AlphaCurve {keys [<at 0 1 RCIM_Cubic>,<at 0.9977 1 RCIM_Cubic>]}
           BlueCurve {keys [<at 0.1175 0.3073 RCIM_Cubic>,<at 0.2682 0.6024>,<at 0.4507 0.4809>,<at 0.5951 0.0816>,<at 0.7691 0.7378>,<at 0.9071 0.2865 RCIM_Cubic>]}
           GreenCurve {keys [<at 0.1175 0.3073 RCIM_Cubic>,<at 0.2682 0.6024>,<at 0.4507 0.4809>,<at 0.5951 0.0816>,<at 0.7691 0.7378>,<at 0.9071 0.2865 RCIM_Cubic>]}
           RedCurve {keys [<at 0.1175 0.3073 RCIM_Cubic>,<at 0.2682 0.6024>,<at 0.4507 0.4809>,<at 0.5951 0.0816>,<at 0.7691 0.7378>,<at 0.9071 0.2865 RCIM_Cubic>]}
          }})
        LerpFactor=$User.VariantWhite
       ))
      "Float (Alpha)"=1
     ))) {isSetParameters=true}
   &ScaleSpriteSize(Scale Factor=&Vector2DFromFloat(Value=&FloatFromCurve(FloatCurve=DataInterface() {propertyValues {Curve {keys [
          {arriveTangent=2.3096, interpMode=RCIM_Cubic, leaveTangent=2.3096, tangentMode=RCTM_User, time=0, value=0}
          {arriveTangent=-0.0952, interpMode=RCIM_Cubic, leaveTangent=-0.0952, tangentMode=RCTM_User, time=0.431, value=0.3973}
          <at 0.9969 0.0123 RCIM_Cubic>
         ]}}})))
   /SetVariables(Particles.Color=&MakeLinearColorFromVectorAndFloat(
     "Vector (RGB)"=&Lerp_Vector(
      A=&Multiply_VectorByFloat(Vector=$Particles.GradientColor, Float=5)
      B=&Multiply_VectorByFloat(Vector=&UniformRangedVector(Minimum=$Particles.GradientColor, Maximum=$Particles.GradientColor), Float=5)
      Alpha=&FloatFromCurve(FloatCurve=DataInterface() {propertyValues {Curve {keys [
           {interpMode=RCIM_Cubic, tangentMode=RCTM_User, time=-0.0056, value=0.0478}
           {arriveTangent=0.0255, interpMode=RCIM_Cubic, leaveTangent=0.0255, tangentMode=RCTM_User, time=0.241, value=0.0789}
           {arriveTangent=0.1228, interpMode=RCIM_Cubic, leaveTangent=0.1228, tangentMode=RCTM_User, time=0.5453, value=0.6822}
           {interpMode=RCIM_Cubic, tangentMode=RCTM_User, time=1.0011, value=0.002}
          ]}}})
     )
     "Float (Alpha)"=&FloatFromCurve(FloatCurve=DataInterface() {propertyValues {Curve {keys [
          {interpMode=RCIM_Cubic, tangentMode=RCTM_User, time=-0.0063, value=0.0184}
          {arriveTangent=1.3376, interpMode=RCIM_Cubic, leaveTangent=1.3376, tangentMode=RCTM_User, time=0.096, value=0.9808}
          {arriveTangent=0.3112, interpMode=RCIM_Cubic, leaveTangent=0.3112, tangentMode=RCTM_User, time=0.5407, value=1.3419}
          {arriveTangent=-3.7119, interpMode=RCIM_Cubic, leaveTangent=-3.7119, tangentMode=RCTM_User, time=0.7242, value=0.3749}
          {interpMode=RCIM_Cubic, tangentMode=RCTM_User, time=1.0011, value=0.002}
         ]}}})
    )) {isSetParameters=true}
   /SetVariables(Particles.Color2=&MakeLinearColorFromVectorAndFloat(
     "Vector (RGB)"=&Lerp_Vector(
      A=&Multiply_VectorByFloat(Vector=$Particles.GradientColor, Float=2)
      B=&Multiply_VectorByFloat(Vector=&UniformRangedVector(Minimum=$Particles.GradientColor, Maximum=$Particles.GradientColor), Float=5)
      Alpha=&FloatFromCurve(FloatCurve=DataInterface() {propertyValues {Curve {keys [
           {interpMode=RCIM_Cubic, tangentMode=RCTM_User, time=-0.0008, value=0.5048}
           {arriveTangent=0.0255, interpMode=RCIM_Cubic, leaveTangent=0.0255, tangentMode=RCTM_User, time=0.0802, value=0.5788}
           {arriveTangent=0.1228, interpMode=RCIM_Cubic, leaveTangent=0.1228, tangentMode=RCTM_User, time=0.5421, value=2.0946}
           {interpMode=RCIM_Cubic, tangentMode=RCTM_User, time=1.0011, value=0.002}
          ]}}})
     )
     "Float (Alpha)"=&FloatFromCurve(
      FloatCurve=DataInterface() {propertyValues {Curve {keys [
          {interpMode=RCIM_Cubic, tangentMode=RCTM_User, time=-0.0063, value=0.0184}
          {arriveTangent=0.3112, interpMode=RCIM_Cubic, leaveTangent=0.3112, tangentMode=RCTM_User, time=0.3734, value=0.0777}
          {arriveTangent=-1.0034, interpMode=RCIM_Cubic, leaveTangent=-1.0034, tangentMode=RCTM_User, time=0.6422, value=0.1616}
          {interpMode=RCIM_Cubic, tangentMode=RCTM_User, time=1.0011, value=0.002}
         ]}}}
      Scale Curve=0.5
     )
    )) {isSetParameters=true}
   &SolveForcesAndVelocity(Clamp Velocity=false, Limit Acceleration=false)
  ]
  rendererClasses [&NiagaraSpriteRendererProperties]
  renderers [{rendererClass=&NiagaraSpriteRendererProperties}]
 }
 Flash_Cheap_Core {
  emitterUpdateScript [
   &EmitterState(
    Enable Distance Culling=true
    Max Distance Response=::KillImmediately
    Module.bUseMaxDistance=true
    Module.bUseMinDistance=false
    Scalability Mode=::NewEnumerator1
   )
   &SpawnBurst_Instantaneous(Spawn Count=1, Use Spawn Probability=false)
  ]
  particleSpawnScript [
   &InitializeParticle(
    Color=&LerpLinearColors(StartColor=<rgba 0 10 6.0323 0.5>, EndColor=<rgba 7 7 7 0.5>, LerpFactor=$User.VariantWhite)
    Lifetime=0.14
    Module.Write Color=true
    Module.Write Lifetime=true
    Module.Write Mass=true
    Module.Write Material Random=true
    Module.Write Position=true
    Module.Write Scale=false
    Module.Write SpriteRotation=true
    Module.Write SpriteSize=true
    Position=&SimulationPosition()
    Sprite Rotation=&UniformRangedFloat(Minimum=0, Maximum=360)
    Sprite Size=&Vector2DFromFloat(Value=&UniformRangedFloat(Minimum=25, Maximum=30))
   )
   &FN_PerParticleLOD(Enable SpriteSize Distance Scaling=true, Sprite Size Scale Far=<2.5 2.5>)
   /SetVariables(Particles.RandomFloat=&UniformRangedFloat(Minimum=0, Maximum=1)) {isSetParameters=true}
   &CameraOffset(Camera Offset Amount=22)
  ]
  particleUpdateScript [
   &ParticleState()
   &ScaleColor(Scale RGB=&MakeVectorFromLinearColorRGB(Color=<rgba 7 2.8407 0.4871 1>), ScaleA=true, ScaleRGB=false)
   &ScaleSpriteSize(Scale Factor=&Vector2DFromFloat(Value=&FloatFromCurve(FloatCurve=DataInterface() {propertyValues {Curve {keys [
          {arriveTangent=2.9657, interpMode=RCIM_Cubic, leaveTangent=2.9657, tangentMode=RCTM_User, time=0, value=0.29}
          {arriveTangent=1.5119, interpMode=RCIM_Cubic, leaveTangent=1.5119, tangentMode=RCTM_User, time=0.15, value=0.6688}
          <at 1 1.03>
         ]}}}))) {module=ScaleSpriteSize001}
   &SubUVAnimation(End Frame=3, SubUV Animation Mode:ENiagara_SubUVLookupMode=::NewEnumerator2) {module=SubUVAnimation001}
   /SetVariables(Particles.LightColor=&MultiplyLinearColor(
     Linear Color=$Particles.Color
     Scale Factor=&MakeVector4_FromFloat(Float=&FloatFromCurve(
       FloatCurve=DataInterface() {propertyValues {Curve {keys [{tangentMode=RCTM_User, time=0.25, value=1}, {tangentMode=RCTM_User, time=1, value=0}]}}}
       Scale Curve=0.5
      ))
    )) {isSetParameters=true}
  ]
  rendererClasses [&NiagaraSpriteRendererProperties, &NiagaraLightRendererProperties]
  renderers [
   {rendererClass=&NiagaraSpriteRendererProperties}
   {rendererClass=&NiagaraLightRendererProperties, rendererIndex=1}
  ]
 }
 Pink_Shards {
  emitterUpdateScript [
   &EmitterState()
   &SpawnBurst_Instantaneous(Spawn Count=&RandomRangeInt(Minimum=8, Maximum=12), Use Spawn Probability=false)
  ]
  particleSpawnScript [
   &FN_SetSpawnGroup_v2(Number of Spawn Groups=2)
   &InitializeParticle(
    Color=&MakeLinearColorFromVectorAndFloat(
     "Vector (RGB)"=&Lerp_Vector(
      A=&MakeVectorFromLinearColorRGB(Color=<rgba 18.1359 0 20 1>)
      B=&MakeVectorFromLinearColorRGB(Color=<rgba 20 20 20 1>)
      Alpha=$User.VariantWhite
     )
     "Float (Alpha)"=&UniformRangedFloat(Minimum=1, Maximum=5)
    )
    Lifetime=&Lerp_Float(
     A=&UniformRangedFloat(Minimum=0.1, Maximum=0.314)
     B=&UniformRangedFloat(Minimum=0.1, Maximum=0.345)
     Alpha=HlslExpression() {hlslExpression=Particles.SpawnGroup}
    )
    Mesh Scale=&Multiply_Vector(
     A=&VectorFromFloat(Value=&RandomRangeFloat(Minimum=0.5, Maximum=1))
     B=&UniformRangedVector(Minimum=<0.025 0.025 0.0067>, Maximum=<0.05 0.05 0.025>)
    )
    Module.Write Color=true
    Module.Write Lifetime=true
    Module.Write Mass=false
    Module.Write Material Random=true
    Module.Write Position=false
    Module.Write Scale=true
    Module.Write SpriteRotation=false
    Module.Write SpriteSize=false
    Sprite Size=<5 5>
   )
   &CylinderLocation(
    Cylinder Height=25
    Cylinder Radius=&Lerp_Float(
     A=&UniformRangedFloat(Minimum=0.35, Maximum=0.6)
     B=&UniformRangedFloat(Minimum=4, Maximum=7)
     Alpha=HlslExpression() {hlslExpression=Particles.SpawnGroup}
    )
    MaskSpawnGroup=false
    Module.Enable Lathe Profile=false
    Module.Override Randomness=false
    Module.Surface Only=true
    Orientation Axis:ENiagaraOrientationAxis=::YAxis
    Surface Only Band Thickness:NiagaraFloat=&Lerp_Float(A=5, B=2, Alpha=HlslExpression() {hlslExpression=Particles.SpawnGroup})
    Use Endcaps In Surface Only Mode=false
   )
   &AddVelocityFromPoint(Module.Use Falloff Distance=false, Velocity Strength=&UniformRangedFloat(Minimum=100, Maximum=500))
   &InheritVelocity(Module.Limit Speed=true, Velocity Scale=<2 2 2>)
   &CameraOffset(Camera Offset Amount=25)
  ]
  particleUpdateScript [
   &ParticleState()
   /Niagara/Modules/Update/Forces/CurlNoiseForce(Noise Frequency=5, Noise Strength=500, Pan Noise=false, Random Seed=1)
   &Drag(Drag=6, Use Linear Drag=true, Use Rotational Drag=false)
   &GravityForce()
   &SolveForcesAndVelocity(Clamp Velocity=false, Limit Acceleration=false)
   &ScaleMeshSize(Scale Factor=&Lerp_Vector(
     A=&VectorFromFloat(Value=&FloatFromCurve(FloatCurve=DataInterface() {propertyValues {Curve {keys [
           <at 0 1>
           {arriveTangent=-0.5431, interpMode=RCIM_Cubic, leaveTangent=-0.5431, tangentMode=RCTM_User, time=0.3, value=1}
           {arriveTangent=-5.5854, interpMode=RCIM_Cubic, leaveTangent=-5.5854, tangentMode=RCTM_User, time=1, value=0}
          ]}}}))
     B=&VectorFromFloat(Value=&FloatFromCurve(FloatCurve=DataInterface() {propertyValues {Curve {keys [
           <at 0 1>
           {arriveTangent=-0.5431, interpMode=RCIM_Cubic, leaveTangent=-0.5431, tangentMode=RCTM_User, time=0.6, value=1}
           {arriveTangent=-5.5854, interpMode=RCIM_Cubic, leaveTangent=-5.5854, tangentMode=RCTM_User, time=1, value=0}
          ]}}}))
     Alpha=HlslExpression() {hlslExpression=Particles.SpawnGroup}
    ))
   &ScaleColor(
    Scale Alpha=&FloatFromCurve(FloatCurve=DataInterface() {propertyValues {Curve {keys [<at 0 1>,<at 1 0>]}}})
    Scale RGB=&VectorFromFloat(Value=&RandomRangeFloat(Minimum=1, Maximum=10))
    ScaleA=true
    ScaleRGB=true
   )
  ]
  rendererClasses [&NiagaraMeshRendererProperties]
  renderers [{rendererClass=&NiagaraMeshRendererProperties, rendererIndex=0}]
 }
 SpikeStreakMesh {
  emitterUpdateScript [
   &EmitterState(
    Enable Distance Culling=true
    Max Distance Response=::KillImmediately
    Module.bUseMaxDistance=true
    Module.bUseMinDistance=false
    Scalability Mode=::NewEnumerator1
   )
   &SpawnBurst_Instantaneous(Spawn Count=&UniformRangedInt(Minimum=8, Maximum=10), Use Spawn Probability=false)
  ]
  particleSpawnScript [
   &InitializeParticle(
    Color=&LerpLinearColors(StartColor=<rgba 0.8789 0 1 0.1>, EndColor=<rgba 1.05 1.05 1.05 0.1>, LerpFactor=$User.VariantWhite)
    Lifetime=&UniformRangedFloat(Minimum=0.125, Maximum=0.135)
    Mesh Scale=&Multiply_VectorByFloat(Vector=&UniformRangedVector(Minimum=<1.5 0.3 0.3>, Maximum=<2.3 0.45 0.45>), Float=0.75)
    Module.Write Color=true
    Module.Write Lifetime=true
    Module.Write Mass=true
    Module.Write Material Random=true
    Module.Write Position=true
    Module.Write Scale=true
    Module.Write SpriteRotation=false
    Module.Write SpriteSize=false
    Position=&SimulationPosition()
   )
   &InitialMeshOrientation(
    Orientation Vector=<0 1 1>
    Rotation=&UniformRangedVector(Minimum=<0 0 0>, Maximum=<1 1 1>)
    Use Orientation Vector=true
    Use Rotation Vector=true
   ) {module=InitialMeshOrientation001}
   &CameraOffset(Camera Offset Amount=12)
  ]
  particleUpdateScript [
   &ParticleState()
   &ScaleMeshSize(Scale Factor=&VectorFromCurve(VectorCurve=DataInterface() {propertyValues {
       bOptimizeLUT=false
       XCurve {keys [{arriveTangent=1.8635, interpMode=RCIM_Cubic, leaveTangent=1.8635, tangentMode=RCTM_User, time=-0.0048, value=0.4}, <at 0.9737 1>]}
       YCurve {keys [
         {arriveTangent=0.1613, interpMode=RCIM_Cubic, leaveTangent=0.1613, tangentMode=RCTM_User, time=0, value=1}
         {arriveTangent=-1.3776, interpMode=RCIM_Cubic, leaveTangent=-1.3776, tangentMode=RCTM_User, time=0.9944, value=0}
        ]}
       ZCurve {keys [<at 0 1 RCIM_Cubic>, {arriveTangent=-1.635, interpMode=RCIM_Cubic, leaveTangent=-1.635, tangentMode=RCTM_User, time=0.9961, value=0}]}
      }}))
   /Game/Effects/Niagara/Modules/Materials/MaterialPanner(Offset:NiagaraFloat=&Add_Float(A=$Emitter.Age, B=$Particles.Initial.MaterialRandom), Speed=-5)
  ]
  rendererClasses [&NiagaraMeshRendererProperties]
  renderers [{rendererClass=&NiagaraMeshRendererProperties, rendererIndex=0}]
 }
}
summary {
 systemName=NS_Pickaxe_TaxiUpgradeMultiColored_Impact
 emitters [
  {emitterName=FN_Sprite001, rendererClasses [&NiagaraSpriteRendererProperties]}
  {emitterName=Corona_v2, rendererClasses [&NiagaraSpriteRendererProperties]}
  {emitterName=FN_Mesh, rendererClasses [&NiagaraMeshRendererProperties]}
  {emitterName=SpikeStreakMesh, rendererClasses [&NiagaraMeshRendererProperties]}
  {emitterName=Pink_Shards, rendererClasses [&NiagaraMeshRendererProperties]}
  {
   emitterName=Flash_Cheap_Core
   rendererClasses [&NiagaraSpriteRendererProperties, &NiagaraLightRendererProperties]
  }
 ]
 userVariables [
  User.SkeletalMesh {
   type=NiagaraDataInterfaceSkeletalMesh
   defaultValue {
    struct=NiagaraExt_VariableValue_DataInterface
    value {dataInterface=&self:NiagaraDataInterfaceSkeletalMesh_0}
   }
  }
  User.VariantMaterialWhite {
   type=MaterialInterface
   description=""
   defaultValue {
    struct=NiagaraExt_VariableValue_Object
    value {object=/BRCosmetics/Effects/Niagara/MasterMaterials/Instances/MI_Light_Cone_Spline_Disco_Inst, objectClass=/Script/Engine.MaterialInterface}
   }
  }
  User.VariantWhite {type=NiagaraFloat, defaultValue=0, description=""}
 ]
}
