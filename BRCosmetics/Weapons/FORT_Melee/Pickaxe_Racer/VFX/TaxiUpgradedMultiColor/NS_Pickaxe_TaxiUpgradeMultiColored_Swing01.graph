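# Niagara swing effect for the multicolor Taxi Upgrade pickaxe, drawing colored ribbon trails and sparkles along the weapon.
# Two short-lived ribbon emitters spawn by distance moved, sampling socket pairs on User.SkeletalMesh for placement and ribbon facing.
# Sparkle sprites spawn around one socket, inherit some velocity and drift with curl noise and flickering alpha.
# User.VariantWhite blends trail and sparkle gradients toward white; corona and glow sprite emitters are disabled.
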NiagaraSystem /BRCosmetics/Weapons/FORT_Melee/Pickaxe_Racer/VFX/TaxiUpgradedMultiColor/NS_Pickaxe_TaxiUpgradeMultiColored_Swing01
dependencies {
 usedDataInterfaces [&NiagaraDataInterfaceColorCurve, &NiagaraDataInterfaceCurve]
 usedDynamicInputs []
 usedModules [&SystemState]
}
emitters {
 AnimTrail_Pink_A {
  emitterUpdateScript [
   &EmitterState()
   &SpawnPerUnit(Module.Use Max Movement Threshold=true, Module.Use Movement Tolerance=false, Spawn Per Unit=100, Use Spawn Probability=false)
  ]
  particleSpawnScript [
   &InitializeRibbon(
    Color=&ColorFromCurve(ColorCurve=DataInterface() {propertyValues {
       AlphaCurve {keys [<at 0.9848 1 RCIM_Cubic>,<at 1 0 RCIM_Cubic>]}
       BlueCurve {keys [<at 0 0.0188 RCIM_Cubic>,<at 0.192 0.6883>,<at 0.3741 1>,<at 0.5822 1>,<at 0.7284 1>,<at 0.9018 1>,<at 1 1 RCIM_Cubic>]}
       GreenCurve {keys [<at 0 1 RCIM_Cubic>,<at 0.192 1>,<at 0.3741 0.5671>,<at 0.5822 0>,<at 0.7284 0>,<at 0.9018 0>,<at 1 1 RCIM_Cubic>]}
       RedCurve {keys [<at 0 0 RCIM_Cubic>,<at 0.192 0>,<at 0.3741 0>,<at 0.5822 0.1983>,<at 0.7284 0.8276>,<at 0.9018 0.8276>,<at 1 1 RCIM_Cubic>]}
      }})
    Lifetime=0.175
    Module.Write Color=true
    Module.Write Lifetime=true
    Module.Write Mass=false
    Module.Write Material Random=true
    Module.Write Position=false
    Module.Write RibbonTwist=false
    Module.Write RibbonWidth=true
    Ribbon Width=30
   )
   &SkeletalMeshLocation(
    Apply To Particle Position=false
    Bone / Socket Index=2
    Inherit Velocity=false
    MaskSpawnGroup=false
    Mesh Sampling Type=::NewEnumerator5
    Skeletal Mesh=$User.SkeletalMesh
    Socket Sampling Mode=::NewEnumerator5
    Use Sampled Mesh Attribute as Ribbon ID=false
   )
   &SkeletalMeshLocation(
    Apply To Particle Position=false
    Bone / Socket Index=3
    Inherit Velocity=false
    MaskSpawnGroup=false
    Mesh Sampling Type=::NewEnumerator5
    Skeletal Mesh=$User.SkeletalMesh
    Socket Sampling Mode=::NewEnumerator5
    Use Sampled Mesh Attribute as Ribbon ID=false
   ) {module=SkeletalMeshLocation001}
   &self:AnimTrail_Spawn_Location(
    Location A:Vector3f=&Add_Vector(A=$Output.SkeletalMeshLocation.SampledPosition, B=<0 0 0>)
    Location B:Vector3f=&Add_Vector(A=$Output.SkeletalMeshLocation001.SampledPosition, B=<0 0 0>)
   ) {module=AnimTrail_Spawn_Location}
   /SetVariables(Particles.RibbonFacing=&Subtract_Vector(A=$Output.SkeletalMeshLocation.SampledPosition, B=$Output.SkeletalMeshLocation001.SampledPosition)) {isSetParameters=true, module=SetVariables_#1}
  ]
  particleUpdateScript [
   &ParticleState()
   &Color(Color=&LerpLinearColors(
     StartColor=&ColorFromCurve(ColorCurve=DataInterface() {propertyValues {
        AlphaCurve {keys [<at 0 0.5 RCIM_Cubic>,<at 1 0 RCIM_Cubic>]}
        BlueCurve {keys [<at 0 0.0188 RCIM_Cubic>,<at 0.192 0.6884>,<at 0.3741 1>,<at 0.5822 1>,<at 0.7284 1>,<at 0.9018 1>,<at 1 1 RCIM_Cubic>]}
        GreenCurve {keys [<at 0 1 RCIM_Cubic>,<at 0.192 1>,<at 0.3741 0.5671>,<at 0.5822 0>,<at 0.7284 0>,<at 0.9018 0>,<at 1 1 RCIM_Cubic>]}
        RedCurve {keys [<at 0 0 RCIM_Cubic>,<at 0.192 0>,<at 0.3741 0>,<at 0.5822 0.1983>,<at 0.7284 0.8276>,<at 0.9018 0.8276>,<at 1 1 RCIM_Cubic>]}
       }})
     EndColor=&ColorFromCurve(ColorCurve=DataInterface() {propertyValues {
        AlphaCurve {keys [
          {arriveTangent=0.2428, interpMode=RCIM_Cubic, time=0.1, value=0.05}
          <at 0.2779 0.0681 RCIM_Cubic>
          {arriveTangent=-0.0735, interpMode=RCIM_Cubic, leaveTangent=-0.0735, time=0.4489, value=0.05}
          <at 0.6472 0.0409 RCIM_Cubic>
          {arriveTangent=-0.1252, interpMode=RCIM_Cubic, leaveTangent=-0.1252, time=0.8337, value=0.05}
          <at 0.9977 -0.003 RCIM_Cubic>
         ]}
        BlueCurve {keys [<at 0 5 RCIM_Cubic>,<at 0.1824 1>,<at 1 10>]}
        GreenCurve {keys [<at 0 5 RCIM_Cubic>,<at 0.1824 1>,<at 1 10>]}
        RedCurve {keys [<at 0 5 RCIM_Cubic>,<at 0.1824 1>,<at 1 10>]}
       }})
     LerpFactor=$User.VariantWhite
    ))
  ]
  rendererClasses [&NiagaraRibbonRendererProperties]
  renderers [{rendererClass=&NiagaraRibbonRendererProperties}]
 }
 AnimTrail_Pink_B {
  emitterUpdateScript [
   &EmitterState()
   &SpawnPerUnit(Module.Use Max Movement Threshold=true, Module.Use Movement Tolerance=false, Spawn Per Unit=100, Use Spawn Probability=false)
  ]
  particleSpawnScript [
   &InitializeRibbon(
    Lifetime=0.175
    Module.Write Color=true
    Module.Write Lifetime=true
    Module.Write Mass=false
    Module.Write Material Random=true
    Module.Write Position=false
    Module.Write RibbonTwist=false
    Module.Write RibbonWidth=true
    Ribbon Width=40
   )
   &SkeletalMeshLocation(
    Apply To Particle Position=false
    Bone / Socket Index=3
    Inherit Velocity=false
    MaskSpawnGroup=false
    Mesh Sampling Type=::NewEnumerator5
    Skeletal Mesh=$User.SkeletalMesh
    Socket Sampling Mode=::NewEnumerator5
    Use Sampled Mesh Attribute as Ribbon ID=false
   )
   &SkeletalMeshLocation(
    Apply To Particle Position=false
    Inherit Velocity=false
    MaskSpawnGroup=false
    Mesh Sampling Type=::NewEnumerator5
    Skeletal Mesh=$User.SkeletalMesh
    Socket Sampling Mode=::NewEnumerator5
    Use Sampled Mesh Attribute as Ribbon ID=false
   ) {module=SkeletalMeshLocation001}
   &self:AnimTrail_Spawn_Location(Location A:Vector3f=$Output.SkeletalMeshLocation.SampledPosition, Location B:Vector3f=$Output.SkeletalMeshLocation001.SampledPosition) {module=AnimTrail_Spawn_Location}
   /SetVariables(Particles.RibbonFacing=&Subtract_Vector(A=$Output.SkeletalMeshLocation.SampledPosition, B=$Output.SkeletalMeshLocation001.SampledPosition)) {isSetParameters=true, module=SetVariables_#1}
  ]
  particleUpdateScript [
   &ParticleState()
   &Color(Color=&LerpLinearColors(
     StartColor=&ColorFromCurve(ColorCurve=DataInterface() {propertyValues {
        AlphaCurve {keys [<at 0 0.5 RCIM_Cubic>,<at 1 0 RCIM_Cubic>]}
        BlueCurve {keys [<at 0 0.0188 RCIM_Cubic>,<at 0.192 0.6884>,<at 0.3741 1>,<at 0.5822 1>,<at 0.7284 1>,<at 0.9018 1>,<at 1 1 RCIM_Cubic>]}
        GreenCurve {keys [<at 0 1 RCIM_Cubic>,<at 0.192 1>,<at 0.3741 0.5671>,<at 0.5822 0>,<at 0.7284 0>,<at 0.9018 0>,<at 1 1 RCIM_Cubic>]}
        RedCurve {keys [<at 0 0 RCIM_Cubic>,<at 0.192 0>,<at 0.3741 0>,<at 0.5822 0.1983>,<at 0.7284 0.8276>,<at 0.9018 0.8276>,<at 1 1 RCIM_Cubic>]}
       }})
     EndColor=&ColorFromCurve(ColorCurve=DataInterface() {propertyValues {
        AlphaCurve {keys [
          {arriveTangent=0.2428, interpMode=RCIM_Cubic, time=0.1, value=0.05}
          <at 0.2779 0.0681 RCIM_Cubic>
          {arriveTangent=-0.0735, interpMode=RCIM_Cubic, leaveTangent=-0.0735, time=0.4489, value=0.05}
          <at 0.6472 0.0409 RCIM_Cubic>
          {arriveTangent=-0.1252, interpMode=RCIM_Cubic, leaveTangent=-0.1252, time=0.8337, value=0.05}
          <at 0.9977 -0.003 RCIM_Cubic>
         ]}
        BlueCurve {keys [<at 0 5 RCIM_Cubic>,<at 0.1824 1>,<at 1 10>]}
        GreenCurve {keys [<at 0 5 RCIM_Cubic>,<at 0.1824 1>,<at 1 10>]}
        RedCurve {keys [<at 0 5 RCIM_Cubic>,<at 0.1824 1>,<at 1 10>]}
       }})
     LerpFactor=$User.VariantWhite
    ))
  ]
  rendererClasses [&NiagaraRibbonRendererProperties]
  renderers [{rendererClass=&NiagaraRibbonRendererProperties}]
 }
 Corona_v2 {
  bEnabled=false
  emitterUpdateScript [
   &EmitterState()
   &SpawnBurst_Instantaneous(Spawn Count=1, Use Spawn Probability=false)
  ]
  particleSpawnScript [/Niagara/Modules/Spawn/Initialization/InitializeParticle(
    Lifetime=0.5
    Mesh Scale=&UniformRangedVector(Minimum=<0.2 0.1 0.1>, Maximum=<1 1 1>)
    Module.Write Color=true
    Module.Write Lifetime=true
    Module.Write Mass=false
    Module.Write Material Random=true
    Module.Write Position=false
    Module.Write Scale=false
    Module.Write SpriteRotation=true
    Module.Write SpriteSize=true
    Sprite Rotation=&NormalizedAngleToDegrees(NormalizedAngle=&UniformRangedFloat(Minimum=0, Maximum=1))
    Sprite Size=&Vector2DFromFloat(Value=80)
   )]
  particleUpdateScript [
   &SkeletalMeshLocation(
    Bone / Socket Index=3
    Inherit Velocity=false
    MaskSpawnGroup=false
    Mesh Sampling Type=::NewEnumerator5
    Skeletal Mesh=$User.SkeletalMesh
    Socket Sampling Mode=::NewEnumerator5
    Use Sampled Mesh Attribute as Ribbon ID=false
   )
   &ParticleState()
   /SetVariables(Particles.GradientColor:Vector3f=&MakeVectorFromLinearColorRGB(Color=&MakeLinearColorFromVectorAndFloat(
      "Vector (RGB)"=&MakeVectorFromLinearColorRGB(Color=&ColorFromCurve(ColorCurve=DataInterface() {propertyValues {
          AlphaCurve {keys [<at -0.0023 1 RCIM_Cubic>,<at 0.9977 1 RCIM_Cubic>]}
          BlueCurve {keys [<at 0.1366 0 RCIM_Cubic>,<at 0.264 0>,<at 0.4507 0>,<at 0.5951 1>,<at 0.7691 1.5>,<at 0.9071 1.5678 RCIM_Cubic>]}
          GreenCurve {keys [<at 0.1366 0.039 RCIM_Cubic>,<at 0.264 0.8559>,<at 0.4507 1>,<at 0.5951 0.6207>,<at 0.7691 0.4532>,<at 0.9071 0 RCIM_Cubic>]}
          RedCurve {keys [<at 0.1366 1 RCIM_Cubic>,<at 0.264 2>,<at 0.4507 0.3535>,<at 0.5951 0.0629>,<at 0.7691 0>,<at 0.9071 2 RCIM_Cubic>]}
         }}))
      "Float (Alpha)"=1
     ))) {isSetParameters=true}
   &ScaleSpriteSize(Scale Factor=&Vector2DFromFloat(Value=&FloatFromCurve(FloatCurve=DataInterface() {propertyValues {Curve {keys [
          {arriveTangent=2.8962, interpMode=RCIM_Cubic, leaveTangent=2.8962, tangentMode=RCTM_User, time=-0.0013, value=0.29}
          {arriveTangent=1.7111, interpMode=RCIM_Cubic, leaveTangent=1.7111, tangentMode=RCTM_User, time=0.15, value=0.66}
          <at 0.9981 1.0141>
         ]}}})))
   /SetVariables(Particles.Color=&MakeLinearColorFromVectorAndFloat(
     "Vector (RGB)"=&Lerp_Vector(
      A=&Multiply_VectorByFloat(Vector=$Particles.GradientColor, Float=1)
      B=&Multiply_VectorByFloat(Vector=&UniformRangedVector(Minimum=$Particles.GradientColor, Maximum=$Particles.GradientColor), Float=2)
      Alpha=&FloatFromCurve(FloatCurve=DataInterface() {propertyValues {Curve {keys [
           {interpMode=RCIM_Cubic, tangentMode=RCTM_User, time=-0.0056, value=0.0478}
           {arriveTangent=0.0255, interpMode=RCIM_Cubic, leaveTangent=0.0255, tangentMode=RCTM_User, time=0.241, value=0.0789}
           {arriveTangent=0.1228, interpMode=RCIM_Cubic, leaveTangent=0.1228, tangentMode=RCTM_User, time=0.5453, value=0.6822}
           {interpMode=RCIM_Cubic, tangentMode=RCTM_User, time=1.0011, value=0.002}
          ]}}})
     )
     "Float (Alpha)"=&FloatFromCurve(
      FloatCurve=DataInterface() {propertyValues {Curve {keys [
          {interpMode=RCIM_Cubic, tangentMode=RCTM_User, time=-0.0063, value=0.0184}
          {arriveTangent=1.3376, interpMode=RCIM_Cubic, leaveTangent=1.3376, tangentMode=RCTM_User, time=0.096, value=0.9808}
          {arriveTangent=0.3112, interpMode=RCIM_Cubic, leaveTangent=0.3112, tangentMode=RCTM_User, time=0.5407, value=1.3419}
          {arriveTangent=-3.7119, interpMode=RCIM_Cubic, leaveTangent=-3.7119, tangentMode=RCTM_User, time=0.7242, value=0.3749}
          {interpMode=RCIM_Cubic, tangentMode=RCTM_User, time=1.0011, value=0.002}
         ]}}}
      Scale Curve=0.25
     )
    )) {isSetParameters=true}
   /SetVariables(Particles.Color2=&MakeLinearColorFromVectorAndFloat(
     "Vector (RGB)"=&Lerp_Vector(
      A=&Multiply_VectorByFloat(Vector=$Particles.GradientColor, Float=1)
      B=&Multiply_VectorByFloat(Vector=&UniformRangedVector(Minimum=$Particles.GradientColor, Maximum=$Particles.GradientColor), Float=2)
      Alpha=&FloatFromCurve(FloatCurve=DataInterface() {propertyValues {Curve {keys [
           {interpMode=RCIM_Cubic, tangentMode=RCTM_User, time=-0.0008, value=0.5048}
           {arriveTangent=0.0255, interpMode=RCIM_Cubic, leaveTangent=0.0255, tangentMode=RCTM_User, time=0.0802, value=0.5788}
           {arriveTangent=0.1228, interpMode=RCIM_Cubic, leaveTangent=0.1228, tangentMode=RCTM_User, time=0.5421, value=2.0946}
           {interpMode=RCIM_Cubic, tangentMode=RCTM_User, time=1.0011, value=0.002}
          ]}}})
     )
     "Float (Alpha)"=&FloatFromCurve(
      FloatCurve=DataInterface() {propertyValues {Curve {keys [
          {interpMode=RCIM_Cubic, tangentMode=RCTM_User, time=-0.0063, value=0.0184}
          {arriveTangent=0.3112, interpMode=RCIM_Cubic, leaveTangent=0.3112, tangentMode=RCTM_User, time=0.3734, value=0.0777}
          {arriveTangent=-1.0034, interpMode=RCIM_Cubic, leaveTangent=-1.0034, tangentMode=RCTM_User, time=0.6422, value=0.1616}
          {interpMode=RCIM_Cubic, tangentMode=RCTM_User, time=1.0011, value=0.002}
         ]}}}
      Scale Curve=0.1
     )
    )) {isSetParameters=true}
   &DynamicMaterialParameters(
    Index 0 Param 1=1
    Index 0 Param 2=0.5
    Index 0 Param 3=0.6
    Index 0 Param 4=1
    Module.Param0WriteEnabled=true
    Module.Param1WriteEnabled=true
    Module.Param2WriteEnabled=true
    Module.Param3WriteEnabled=true
   )
  ]
  rendererClasses [&NiagaraSpriteRendererProperties]
  renderers [{rendererClass=&NiagaraSpriteRendererProperties}]
 }
 FN_Sprite001 {
  bEnabled=false
  emitterUpdateScript [
   &EmitterState()
   &SpawnBurst_Instantaneous(Spawn Count=1, Use Spawn Probability=false)
  ]
  particleSpawnScript [
   &InitializeParticle(
    Lifetime=0.4
    Module.Write Color=true
    Module.Write Lifetime=true
    Module.Write Mass=false
    Module.Write Material Random=true
    Module.Write Position=false
    Module.Write Scale=false
    Module.Write SpriteRotation=true
    Module.Write SpriteSize=true
    Position=&SimulationPosition()
    Sprite Rotation=&UniformRangedFloat(Minimum=0, Maximum=360)
    Sprite Size=&Vector2DFromFloat(Value=400)
   )
   &CameraOffset(Camera Offset Amount=30)
  ]
  particleUpdateScript [
   &SkeletalMeshLocation(
    Bone / Socket Index=3
    Inherit Velocity=false
    MaskSpawnGroup=false
    Mesh Sampling Type=::NewEnumerator5
    Skeletal Mesh=$User.SkeletalMesh
    Socket Sampling Mode=::NewEnumerator5
    Use Sampled Mesh Attribute as Ribbon ID=false
   )
   &ParticleState()
   /SetVariables(Particles.GradientColor:Vector3f=&MakeVectorFromLinearColorRGB(Color=&MakeLinearColorFromVectorAndFloat(
      "Vector (RGB)"=&MakeVectorFromLinearColorRGB(Color=&ColorFromCurve(ColorCurve=DataInterface() {propertyValues {
          AlphaCurve {keys [<at 0.9977 0.1 RCIM_Cubic>]}
          BlueCurve {keys [<at 0.1366 0 RCIM_Cubic>,<at 0.264 0>,<at 0.4507 0>,<at 0.5951 1>,<at 0.7691 1.5>,<at 0.9071 1.5678 RCIM_Cubic>]}
          GreenCurve {keys [<at 0.1366 0.039 RCIM_Cubic>,<at 0.264 0.8559>,<at 0.4507 1>,<at 0.5951 0.6207>,<at 0.7691 0.4532>,<at 0.9071 0 RCIM_Cubic>]}
          RedCurve {keys [<at 0.1366 1 RCIM_Cubic>,<at 0.264 2>,<at 0.4507 0.3535>,<at 0.5951 0.0629>,<at 0.7691 0>,<at 0.9071 2 RCIM_Cubic>]}
         }}))
      "Float (Alpha)"=1
     ))) {isSetParameters=true}
   &ScaleSpriteSize(Scale Factor=&Vector2DFromFloat(Value=&FloatFromCurve(FloatCurve=DataInterface() {propertyValues {Curve {keys [
          {arriveTangent=2.3096, interpMode=RCIM_Cubic, leaveTangent=2.3096, tangentMode=RCTM_User, time=0, value=0}
          {arriveTangent=-0.0952, interpMode=RCIM_Cubic, leaveTangent=-0.0952, tangentMode=RCTM_User, time=0.431, value=0.3973}
          <at 0.9969 0.0123 RCIM_Cubic>
         ]}}})))
   /SetVariables(Particles.Color=&MakeLinearColorFromVectorAndFloat(
     "Vector (RGB)"=&Lerp_Vector(
      A=&Multiply_VectorByFloat(Vector=$Particles.GradientColor, Float=5)
      B=&Multiply_VectorByFloat(Vector=&UniformRangedVector(Minimum=$Particles.GradientColor, Maximum=$Particles.GradientColor), Float=5)
      Alpha=&FloatFromCurve(FloatCurve=DataInterface() {propertyValues {Curve {keys [
           {interpMode=RCIM_Cubic, tangentMode=RCTM_User, time=-0.0056, value=0.0478}
           {arriveTangent=0.0255, interpMode=RCIM_Cubic, leaveTangent=0.0255, tangentMode=RCTM_User, time=0.241, value=0.0789}
           {arriveTangent=0.1228, interpMode=RCIM_Cubic, leaveTangent=0.1228, tangentMode=RCTM_User, time=0.5453, value=0.6822}
           {interpMode=RCIM_Cubic, tangentMode=RCTM_User, time=1.0011, value=0.002}
          ]}}})
     )
     "Float (Alpha)"=&FloatFromCurve(FloatCurve=DataInterface() {propertyValues {Curve {keys [
          {interpMode=RCIM_Cubic, tangentMode=RCTM_User, time=-0.0063, value=0.0184}
          {arriveTangent=1.3376, interpMode=RCIM_Cubic, leaveTangent=1.3376, tangentMode=RCTM_User, time=0.096, value=0.9808}
          {arriveTangent=0.3112, interpMode=RCIM_Cubic, leaveTangent=0.3112, tangentMode=RCTM_User, time=0.5407, value=1.3419}
          {arriveTangent=-3.7119, interpMode=RCIM_Cubic, leaveTangent=-3.7119, tangentMode=RCTM_User, time=0.7242, value=0.3749}
          {interpMode=RCIM_Cubic, tangentMode=RCTM_User, time=1.0011, value=0.002}
         ]}}})
    )) {isSetParameters=true}
   /SetVariables(Particles.Color2=&MakeLinearColorFromVectorAndFloat(
     "Vector (RGB)"=&Lerp_Vector(
      A=&Multiply_VectorByFloat(Vector=$Particles.GradientColor, Float=2)
      B=&Multiply_VectorByFloat(Vector=&UniformRangedVector(Minimum=$Particles.GradientColor, Maximum=$Particles.GradientColor), Float=5)
      Alpha=&FloatFromCurve(
       FloatCurve=DataInterface() {propertyValues {Curve {keys [
           {interpMode=RCIM_Cubic, tangentMode=RCTM_User, time=-0.0008, value=0.5048}
           {arriveTangent=0.0255, interpMode=RCIM_Cubic, leaveTangent=0.0255, tangentMode=RCTM_User, time=0.0802, value=0.5788}
           {arriveTangent=0.1228, interpMode=RCIM_Cubic, leaveTangent=0.1228, tangentMode=RCTM_User, time=0.5421, value=2.0946}
           {interpMode=RCIM_Cubic, tangentMode=RCTM_User, time=1.0011, value=0.002}
          ]}}}
       Scale Curve=0.15
      )
     )
     "Float (Alpha)"=0.05
    )) {isSetParameters=true}
   &SolveForcesAndVelocity(Clamp Velocity=false, Limit Acceleration=false)
  ]
  rendererClasses [&NiagaraSpriteRendererProperties]
  renderers [{rendererClass=&NiagaraSpriteRendererProperties}]
 }
 Sparkles {
  emitterUpdateScript [
   &EmitterState()
   &SpawnRate(
    Spawn Probability=&FloatFromCurve(FloatCurve=DataInterface() {propertyValues {Curve {keys [<at 0 0>,<at 0.1 1>]}}}, CurveIndex=$Emitter.LoopedAge)
    SpawnRate=170
    Use Spawn Probability=true
   )
  ]
  particleSpawnScript [
   &SkeletalMeshLocation(
    Bone / Socket Index=3
    Inherit Velocity=false
    MaskSpawnGroup=false
    Mesh Sampling Type=::NewEnumerator5
    Skeletal Mesh=$User.SkeletalMesh
    Socket Sampling Mode=::NewEnumerator5
    Use Sampled Mesh Attribute as Ribbon ID=false
   )
   &InitializeParticle(
    Color=<rgba 0.25 0.25 1 0.5>
    Lifetime=&RandomRangeFloat(Minimum=0.3, Maximum=0.5)
    Mesh Scale=<1.5 1.5 1.5>
    Module.Write Color=true
    Module.Write Lifetime=true
    Module.Write Mass=false
    Module.Write Material Random=true
    Module.Write Position=false
    Module.Write Scale=false
    Module.Write SpriteRotation=true
    Module.Write SpriteSize=true
    Position=&SimulationPosition()
    Sprite Rotation=&UniformRangedFloat(Minimum=-15, Maximum=15)
    Sprite Size=&UniformRangedVector2D(
     Minimum=&Vector2DFromFloat(Value=3)
     Maximum=&Vector2DFromFloat(Value=8)
    )
   )
   &SphereLocation(
    MaskSpawnGroup=false
    Override Seed=false
    Sphere Radius=9
    Surface Only=false
    UseNonUniformScale=false
    UseOffset=false
    UseSphereOrientation=false
   ) {module=SphereLocation001}
   &InheritVelocity(Module.Limit Speed=true, Velocity Scale=&VectorFromFloat(Value=0.3))
  ]
  particleUpdateScript [
   &ParticleState()
   &Color(Color=&LerpLinearColors(
     StartColor=&ColorFromCurve(ColorCurve=DataInterface() {propertyValues {
        AlphaCurve {keys [
          {arriveTangent=0.2428, interpMode=RCIM_Cubic, time=0.1, value=1}
          {arriveTangent=-0.0163, interpMode=RCIM_Cubic, leaveTangent=-0.0163, time=0.2779, value=0.0681}
          {arriveTangent=-0.0735, interpMode=RCIM_Cubic, leaveTangent=-0.0735, time=0.4489, value=0.9943}
          {arriveTangent=0.0148, interpMode=RCIM_Cubic, leaveTangent=0.0148, time=0.6472, value=0.0409}
          {arriveTangent=-0.1252, interpMode=RCIM_Cubic, leaveTangent=-0.1252, time=0.8337, value=1}
          <at 0.9977 -0.003 RCIM_Cubic>
         ]}
        BlueCurve {keys [<at 0 3.5 RCIM_Cubic>,<at 0.1824 10>,<at 1 0.8483>]}
        GreenCurve {keys [<at 0 10 RCIM_Cubic>,<at 0.1824 0.25>,<at 1 0>]}
        RedCurve {keys [<at 0 0.25 RCIM_Cubic>,<at 0.1824 0.25>,<at 1 20>]}
       }})
     EndColor=&ColorFromCurve(ColorCurve=DataInterface() {propertyValues {
        AlphaCurve {keys [
          {arriveTangent=0.2428, interpMode=RCIM_Cubic, time=0.1, value=1}
          {arriveTangent=-0.0163, interpMode=RCIM_Cubic, leaveTangent=-0.0163, time=0.2779, value=0.0681}
          {arriveTangent=-0.0735, interpMode=RCIM_Cubic, leaveTangent=-0.0735, time=0.4489, value=0.9943}
          {arriveTangent=0.0148, interpMode=RCIM_Cubic, leaveTangent=0.0148, time=0.6472, value=0.0409}
          {arriveTangent=-0.1252, interpMode=RCIM_Cubic, leaveTangent=-0.1252, time=0.8337, value=1}
          <at 0.9977 -0.003 RCIM_Cubic>
         ]}
        BlueCurve {keys [<at 0 10 RCIM_Cubic>,<at 0.1824 1>,<at 1 20>]}
        GreenCurve {keys [<at 0 10 RCIM_Cubic>,<at 0.1824 1>,<at 1 20>]}
        RedCurve {keys [<at 0 10 RCIM_Cubic>,<at 0.1824 1>,<at 1 20>]}
       }})
     LerpFactor=$User.VariantWhite
    ))
   &ScaleSpriteSize(Scale Factor=&Vector2DFromFloat(Value=&FloatFromCurve(
      FloatCurve=DataInterface() {propertyValues {Curve {keys [
          {tangentMode=RCTM_User, time=0, value=1}
          {arriveTangent=-1, interpMode=RCIM_Cubic, leaveTangent=-1, time=0.2, value=1}
          {interpMode=RCIM_Cubic, leaveTangent=-8.0089, time=1, value=0}
         ]}}}
      Scale Curve=&RandomRangeFloat(Minimum=1, Maximum=2)
     )))
   /Niagara/Modules/Update/Forces/CurlNoiseForce(Noise Strength=200, Pan Noise=false)
   &Drag(Drag=&RandomRangeFloat(Minimum=0, Maximum=3), Use Linear Drag=true, Use Rotational Drag=false)
   &SolveForcesAndVelocity(Clamp Velocity=false, Limit Acceleration=false)
  ]
  rendererClasses [&NiagaraSpriteRendererProperties]
  renderers [{rendererClass=&NiagaraSpriteRendererProperties}]
 }
}
summary {
 systemName=NS_Pickaxe_TaxiUpgradeMultiColored_Swing01
 emitters [
  {emitterName=AnimTrail_Pink_A, rendererClasses [&NiagaraRibbonRendererProperties]}
  {emitterName=AnimTrail_Pink_B, rendererClasses [&NiagaraRibbonRendererProperties]}
  {bEnabled=false, emitterName=FN_Sprite001, rendererClasses [&NiagaraSpriteRendererProperties]}
  {bEnabled=false, emitterName=Corona_v2, rendererClasses [&NiagaraSpriteRendererProperties]}
  {emitterName=Sparkles, rendererClasses [&NiagaraSpriteRendererProperties]}
 ]
 userVariables [
  User.SkeletalMesh {
   type=NiagaraDataInterfaceSkeletalMesh
   defaultValue {
    struct=NiagaraExt_VariableValue_DataInterface
    value {dataInterface=&self:NiagaraDataInterfaceSkeletalMesh_0}
   }
  }
  User.VariantWhite {type=NiagaraFloat, defaultValue=0, description=""}
 ]
}
